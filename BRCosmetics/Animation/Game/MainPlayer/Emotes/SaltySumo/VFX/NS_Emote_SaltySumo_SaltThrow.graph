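# Niagara effect for the salt throw in the Salty Sumo emote, spraying sprite grains from a skeletal mesh plus a user offset.
# Four emitters spawn on decaying rate curves over brief bursts, launching grains with cone and upward velocity under gravity.
# The GPU emitter adds delayed collision for long-lived, bouncing grains, while another varies spawn rate by platform quality.
# Color scales with velocity, and two emitters also scale blue by a time-of-day value.
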NiagaraSystem /BRCosmetics/Animation/Game/MainPlayer/Emotes/SaltySumo/VFX/NS_Emote_SaltySumo_SaltThrow
dependencies {
 usedDataInterfaces [&NiagaraDataInterfaceCurve, &NiagaraDataInterfacePlatformSet, &NiagaraDataInterfaceSkeletalMesh]
 usedDynamicInputs [&FNTimeOfDayScalar, &FN_FrontEndFloatModifier]
 usedModules [
  &self:SystemSpawnScript.SystemScriptSource.SystemScriptGraph.NiagaraNodeAssignment_1.SetVariables_#1
  &SystemState
  &self:SystemSpawnScript.SystemScriptSource.SystemScriptGraph.NiagaraNodeAssignment_0.SetVariables_#2
 ]
}
emitters {
 SimpleSpriteBurst {
  simTarget=GPUComputeSim
  emitterUpdateScript [
   &EmitterState(Life Cycle Mode=::NewEnumerator1, Loop Behavior=::NewEnumerator1, Loop Duration=0.3, UseLoopDelay=false)
   &SpawnRate(
    SpawnRate=&FloatFromCurve(
     FloatCurve=DataInterface() {propertyValues {Curve {keys [{interpMode=RCIM_Cubic, tangentMode=RCTM_User, time=0, value=1}, {arriveTangent=-2.7489, tangentMode=RCTM_User, time=1, value=0}]}}}
     CurveIndex=$Emitter.NormalizedLoopAge
     Scale Curve=1200
    )
    Use Spawn Probability=false
   )
  ]
  particleSpawnScript [
   &InitializeParticle(
    Color=&RandomRangeLinearColor(Minimum=<rgba 0.8 1 1 1>, Maximum=<rgba 0.5 0 1 1>)
    Lifetime=&RandomRangeFloat(Minimum=3, Maximum=8)
    Sprite Rotation Mode=::NewEnumerator2
    Sprite Size Mode=::NewEnumerator1
    Uniform Sprite Size Max=2
    Uniform Sprite Size Min=0.3
    UsePositionOffset=true
    Write Material Random=true
   )
   /SetVariables() {isSetParameters=true, module=SetVariables_#3}
   &AddVelocity(
    Cone Angle=&RandomRangeFloat(Minimum=40, Maximum=110)
    Cone Axis=<1 0 -1>
    Module.UseConeFalloff=false
    Module.UseVelDistribution=false
    Override Seed=false
    Velocity Mode=::NewEnumerator2
    Velocity Speed=&RandomRangeFloat(
     Minimum=&FloatFromCurve(
      FloatCurve=DataInterface() {propertyValues {Curve {keys [<at 0.1806 0.8>,<at 0.5183 1>]}}}
      CurveIndex=$Emitter.NormalizedLoopAge
      Scale Curve=60
     )
     Maximum=&MakeCustomFloatFromBool(
      Bool=&SetBoolByFloatComparison(A=&RandomRangeFloat(Minimum=0, Maximum=1), B=0.6, Comparison Type=::NewEnumerator6)
      True Float=&RandomRangeFloat(Minimum=30, Maximum=60)
      False Float=140
     )
    )
   )
   &AddVelocity(
    Rotation Coordinate Space=::World
    Velocity=&RandomRangeVector(Minimum=<0 0 75>, Maximum=<0 0 225>)
    Velocity Speed Scale=&MakeCustomFloatFromBool(
     Bool=&SetBoolByFloatComparison(A=&RandomRangeFloat(Minimum=0, Maximum=1), B=0.6, Comparison Type=::NewEnumerator6)
     True Float=&RandomRangeFloat(Minimum=0.34, Maximum=0.42)
     False Float=1
    )
   ) {module=AddVelocity001}
   &SkeletalMeshLocation(
    Bone Sampling Mode=::NewEnumerator0
    MaskSpawnGroup=false
    Orientation Sampling=::NewEnumerator1
    OverrideRandomSeed=false
    Position Sampling=::NewEnumerator1
    Sampled Position Offset=$User.Offset
    Sampled Position Offset Coordinate Space:ENiagaraSkelMeshTransforms=::NewEnumerator2
    Skeletal Mesh=$System.Skeletal Mesh
   )
  ]
  particleUpdateScript [
   &ParticleState()
   &ScaleSpriteSize(Uniform Curve Sprite Scale=DataInterface() {propertyValues {Curve {keys [
        <at 0.005 0 RCIM_Cubic>
        {interpMode=RCIM_Cubic, tangentMode=RCTM_User, time=0.01, value=1}
        {arriveTangent=-1.0101, interpMode=RCIM_Cubic, leaveTangent=-1.0101, time=0.4, value=1}
        {interpMode=RCIM_Cubic, tangentMode=RCTM_User, time=1, value=0}
       ]}}})
   &ScaleColor(
    Scale Alpha=&FloatFromCurve(FloatCurve=DataInterface() {propertyValues {Curve {keys [<at -0.0046 1.0113 RCIM_Cubic>,<at 1.0011 0.002 RCIM_Cubic>]}}})
    Scale RGB=&MakeVector(
     X=1
     Y=&Lerp_Float(
      A=0
      B=&FloatFromCurve(FloatCurve=DataInterface() {propertyValues {Curve {keys [<at 0 1>,<at 0.0957 0>]}}})
      Alpha=&MakeFloatFromVector(Vector=$Particles.Velocity, Channel=::NewEnumerator0)
     )
     Z=1
    )
    ScaleA=true
    ScaleRGB=true
   )
   &GravityForce(Gravity=&RandomRangeVector(Minimum=<0 0 -980>, Maximum=<0 0 -700>))
   &Collision(
    Age Colliding Particles=false
    Collision Enabled=&SetBoolByFloatComparison(A=0.1, B=$Particles.Age, Comparison Type=::NewEnumerator6)
    Friction:NiagaraFloat=&RandomRangeFloat(Minimum=1, Maximum=1.4)
    Kill Offscreen Particles=true
    KillOnCollision=false
    Module.Control Rest State Via Surface Normal=false
    Module.EnableMaxCollisionCount=false
    Module.Randomize Collision Normal Vector=false
    Particle Radius Scale=3
    Restitution=&RandomRangeFloat(Minimum=0.15, Maximum=0.35)
   )
   &SolveForcesAndVelocity(Clamp Velocity=false, Limit Acceleration=false)
   &ScaleColor(
    Scale RGB=&MakeVector(X=1, Y=1, Z=&Lerp_Float(A=1, B=0, Alpha=$System.TOD))
    ScaleA=true
    ScaleRGB=true
   ) {module=ScaleColor001}
  ]
  rendererClasses [&NiagaraSpriteRendererProperties]
  renderers [{rendererClass=&NiagaraSpriteRendererProperties}]
 }
 SimpleSpriteBurst001 {
  emitterUpdateScript [
   &EmitterState(Life Cycle Mode=::NewEnumerator1, Loop Behavior=::NewEnumerator1, Loop Duration=0.3, UseLoopDelay=false)
   &SpawnRate(
    SpawnRate=&FloatFromCurve(
     FloatCurve=DataInterface() {propertyValues {Curve {keys [{interpMode=RCIM_Cubic, tangentMode=RCTM_User, time=0, value=1}, {arriveTangent=-2.7489, tangentMode=RCTM_User, time=1, value=0}]}}}
     CurveIndex=$Emitter.NormalizedLoopAge
     Scale Curve=500
    )
    Use Spawn Probability=false
   )
  ]
  particleSpawnScript [
   &InitializeParticle(
    Color=&RandomRangeLinearColor(Minimum=<rgba 1 1 0 1>, Maximum=<rgba 0.8 1 1 1>)
    Lifetime=&RandomRangeFloat(Minimum=0.23, Maximum=0.7)
    Sprite Rotation Mode=::NewEnumerator2
    Sprite Size Mode=::NewEnumerator1
    Uniform Sprite Size Max=1.6
    Uniform Sprite Size Min=0.4
    UsePositionOffset=false
    Write Material Random=true
   )
   &AddVelocity(
    Cone Angle=50
    Cone Axis=<0 1 0>
    Module.UseConeFalloff=false
    Module.UseVelDistribution=false
    Override Seed=false
    Velocity Mode=::NewEnumerator2
    Velocity Speed=&RandomRangeFloat(Minimum=40, Maximum=110)
   )
   &AddVelocity(
    Rotation Coordinate Space=::World
    Velocity=&RandomRangeVector(Minimum=<45 45 110>, Maximum=<-45 -45 160>)
   ) {module=AddVelocity001}
   &SkeletalMeshLocation(
    Bone Sampling Mode=::NewEnumerator0
    MaskSpawnGroup=false
    Orientation Sampling=::NewEnumerator1
    OverrideRandomSeed=false
    Position Sampling=::NewEnumerator1
    Sampled Position Offset=$User.Offset
    Sampled Position Offset Coordinate Space:ENiagaraSkelMeshTransforms=::NewEnumerator2
    Skeletal Mesh=$System.Skeletal Mesh
   )
  ]
  particleUpdateScript [
   &ParticleState()
   &ScaleSpriteSize(Uniform Curve Sprite Scale=DataInterface() {propertyValues {Curve {keys [<at 0.05 0>, <at 0.1344 1>, {tangentMode=RCTM_User, time=0.3136, value=1}, {tangentMode=RCTM_User, time=1, value=0}]}}})
   &ScaleColor(
    Scale Alpha=&FloatFromCurve(FloatCurve=DataInterface() {propertyValues {Curve {keys [<at -0.0046 1.0113 RCIM_Cubic>,<at 1.0011 0.002 RCIM_Cubic>]}}})
    Scale RGB=&MakeVector(
     X=1
     Y=&Lerp_Float(
      A=0
      B=&FloatFromCurve(FloatCurve=DataInterface() {propertyValues {Curve {keys [<at 0 1>,<at 0.08 0>]}}})
      Alpha=&MakeFloatFromVector(Vector=$Particles.Velocity, Channel=::NewEnumerator0)
     )
     Z=1
    )
    ScaleA=true
    ScaleRGB=true
   )
   &GravityForce()
   &SolveForcesAndVelocity(Clamp Velocity=false, Limit Acceleration=false)
  ]
  rendererClasses [&NiagaraSpriteRendererProperties]
  renderers [{rendererClass=&NiagaraSpriteRendererProperties}]
 }
 SimpleSpriteBurst002 {
  emitterUpdateScript [
   &EmitterState(Life Cycle Mode=::NewEnumerator1, Loop Behavior=::NewEnumerator1, Loop Duration=0.3, UseLoopDelay=false)
   &SpawnRate(
    SpawnRate=&FloatFromCurve(
     FloatCurve=DataInterface() {propertyValues {Curve {keys [{interpMode=RCIM_Cubic, tangentMode=RCTM_User, time=0, value=1}, {arriveTangent=-2.7489, tangentMode=RCTM_User, time=1, value=0}]}}}
     CurveIndex=$Emitter.NormalizedLoopAge
     Scale Curve=&MakeCustomFloatFromBool(
      Bool=&IsPlatformSetActive(New Platform Set=DataInterface() {propertyValues {Platforms {qualityLevelMask=-31}}})
      True Float=200
      False Float=&MakeCustomFloatFromBool(
       Bool=&IsPlatformSetActive(New Platform Set=DataInterface() {propertyValues {Platforms {qualityLevelMask=-30}}})
       True Float=1000
       False Float=5000
      )
     )
    )
    Use Spawn Probability=false
   ) {module=SpawnRate001}
  ]
  particleSpawnScript [
   &InitializeParticle(
    Color=&RandomRangeLinearColor(Minimum=<rgba 0.7 1 0 1>, Maximum=<rgba 0.5 1 1 1>)
    Lifetime=&RandomRangeFloat(Minimum=0.9, Maximum=0.6)
    Sprite Rotation Mode=::NewEnumerator2
    Sprite Size Mode=::NewEnumerator1
    Uniform Sprite Size Max=2
    Uniform Sprite Size Min=0.1
    UsePositionOffset=false
    Write Material Random=true
   )
   &AddVelocity(
    Cone Angle=&RandomRangeFloat(Minimum=40, Maximum=110)
    Cone Axis=<1 0 -1>
    Module.UseConeFalloff=false
    Module.UseVelDistribution=false
    Override Seed=false
    Velocity Mode=::NewEnumerator2
    Velocity Speed=&RandomRangeFloat(
     Minimum=&FloatFromCurve(
      FloatCurve=DataInterface() {propertyValues {Curve {keys [<at 0.1806 0.8>,<at 0.5183 1>]}}}
      CurveIndex=$Emitter.NormalizedLoopAge
      Scale Curve=50
     )
     Maximum=&MakeCustomFloatFromBool(
      Bool=&SetBoolByFloatComparison(A=&RandomRangeFloat(Minimum=0, Maximum=1), B=0.6, Comparison Type=::NewEnumerator6)
      True Float=&RandomRangeFloat(Minimum=30, Maximum=60)
      False Float=140
     )
    )
   )
   &AddVelocity(
    Rotation Coordinate Space=::World
    Velocity=&RandomRangeVector(Minimum=<0 0 75>, Maximum=<0 0 225>)
    Velocity Speed Scale=&MakeCustomFloatFromBool(
     Bool=&SetBoolByFloatComparison(A=&RandomRangeFloat(Minimum=0, Maximum=1), B=0.6, Comparison Type=::NewEnumerator6)
     True Float=&RandomRangeFloat(Minimum=0.34, Maximum=0.42)
     False Float=1
    )
   ) {module=AddVelocity001}
   &SkeletalMeshLocation(
    Bone Sampling Mode=::NewEnumerator0
    MaskSpawnGroup=false
    Orientation Sampling=::NewEnumerator1
    OverrideRandomSeed=false
    Position Sampling=::NewEnumerator1
    Sampled Position Offset=$User.Offset
    Sampled Position Offset Coordinate Space:ENiagaraSkelMeshTransforms=::NewEnumerator2
    Skeletal Mesh=$System.Skeletal Mesh
   )
  ]
  particleUpdateScript [
   &ParticleState()
   &ScaleSpriteSize(Uniform Curve Sprite Scale=DataInterface() {propertyValues {Curve {keys [
        <at 0.05 0 RCIM_Cubic>
        {interpMode=RCIM_Cubic, tangentMode=RCTM_User, time=0.1172, value=1}
        {interpMode=RCIM_Cubic, tangentMode=RCTM_User, time=1, value=0}
       ]}}})
   &ScaleColor(
    Scale Alpha=&FloatFromCurve(FloatCurve=DataInterface() {propertyValues {Curve {keys [<at -0.0046 1.0113 RCIM_Cubic>,<at 1.0011 0.002 RCIM_Cubic>]}}})
    Scale RGB=&MakeVector(
     X=1
     Y=&Lerp_Float(
      A=0
      B=&FloatFromCurve(FloatCurve=DataInterface() {propertyValues {Curve {keys [<at 0 1>,<at 0.08 0>]}}})
      Alpha=&MakeFloatFromVector(Vector=$Particles.Velocity, Channel=::NewEnumerator0)
     )
     Z=1
    )
    ScaleA=true
    ScaleRGB=true
   )
   &GravityForce()
   &SolveForcesAndVelocity(Clamp Velocity=false, Limit Acceleration=false)
  ]
  rendererClasses [&NiagaraSpriteRendererProperties]
  renderers [{rendererClass=&NiagaraSpriteRendererProperties}]
 }
 SimpleSpriteBurst003 {
  emitterUpdateScript [
   &EmitterState(Life Cycle Mode=::NewEnumerator1, Loop Behavior=::NewEnumerator1, Loop Duration=0.3, UseLoopDelay=false)
   &SpawnRate(
    SpawnRate=&FloatFromCurve(
     FloatCurve=DataInterface() {propertyValues {Curve {keys [{interpMode=RCIM_Cubic, tangentMode=RCTM_User, time=0, value=1}, {arriveTangent=-2.7489, tangentMode=RCTM_User, time=1, value=0}]}}}
     CurveIndex=$Emitter.NormalizedLoopAge
     Scale Curve=575
    )
    Use Spawn Probability=false
   )
  ]
  particleSpawnScript [
   &InitializeParticle(
    Color=&RandomRangeLinearColor(Minimum=<rgba 0.5 1 1 1>, Maximum=<rgba 0.5 0 1 1>)
    Lifetime=&RandomRangeFloat(Minimum=1, Maximum=2)
    Sprite Rotation Mode=::NewEnumerator2
    Sprite Size Mode=::NewEnumerator1
    Uniform Sprite Size Max=2
    Uniform Sprite Size Min=0.5
    UsePositionOffset=true
    Write Material Random=true
   )
   /SetVariables() {isSetParameters=true, module=SetVariables_#3}
   &AddVelocity(
    Cone Angle=&RandomRangeFloat(Minimum=40, Maximum=110)
    Cone Axis=<1 0 -1>
    Module.UseConeFalloff=false
    Module.UseVelDistribution=false
    Override Seed=false
    Velocity Mode=::NewEnumerator2
    Velocity Speed=&RandomRangeFloat(
     Minimum=&FloatFromCurve(
      FloatCurve=DataInterface() {propertyValues {Curve {keys [<at 0.1806 0.8>,<at 0.5183 1>]}}}
      CurveIndex=$Emitter.NormalizedLoopAge
      Scale Curve=60
     )
     Maximum=&MakeCustomFloatFromBool(
      Bool=&SetBoolByFloatComparison(A=&RandomRangeFloat(Minimum=0, Maximum=1), B=0.6, Comparison Type=::NewEnumerator6)
      True Float=&RandomRangeFloat(Minimum=30, Maximum=60)
      False Float=140
     )
    )
   )
   &AddVelocity(
    Rotation Coordinate Space=::World
    Velocity=&RandomRangeVector(Minimum=<0 0 75>, Maximum=<0 0 225>)
    Velocity Speed Scale=&MakeCustomFloatFromBool(
     Bool=&SetBoolByFloatComparison(A=&RandomRangeFloat(Minimum=0, Maximum=1), B=0.6, Comparison Type=::NewEnumerator6)
     True Float=&RandomRangeFloat(Minimum=0.34, Maximum=0.42)
     False Float=1
    )
   ) {module=AddVelocity001}
   &SkeletalMeshLocation(
    Bone Sampling Mode=::NewEnumerator0
    MaskSpawnGroup=false
    Orientation Sampling=::NewEnumerator1
    OverrideRandomSeed=false
    Position Sampling=::NewEnumerator1
    Sampled Position Offset=$User.Offset
    Sampled Position Offset Coordinate Space:ENiagaraSkelMeshTransforms=::NewEnumerator2
    Skeletal Mesh=$System.Skeletal Mesh
   )
  ]
  particleUpdateScript [
   &ParticleState()
   &ScaleSpriteSize(Uniform Curve Sprite Scale=DataInterface() {propertyValues {Curve {keys [
        <at 0.005 0 RCIM_Cubic>
        {interpMode=RCIM_Cubic, tangentMode=RCTM_User, time=0.01, value=1}
        {arriveTangent=-1.0101, interpMode=RCIM_Cubic, leaveTangent=-1.0101, time=0.4, value=1}
        {interpMode=RCIM_Cubic, tangentMode=RCTM_User, time=1, value=0}
       ]}}})
   &ScaleColor(
    Scale Alpha=&FloatFromCurve(FloatCurve=DataInterface() {propertyValues {Curve {keys [<at -0.0046 1.0113 RCIM_Cubic>,<at 1.0011 0.002 RCIM_Cubic>]}}})
    Scale RGB=&MakeVector(
     X=1
     Y=&Lerp_Float(
      A=0
      B=&FloatFromCurve(FloatCurve=DataInterface() {propertyValues {Curve {keys [<at 0 1>,<at 0.0957 0>]}}})
      Alpha=&MakeFloatFromVector(Vector=$Particles.Velocity, Channel=::NewEnumerator0)
     )
     Z=1
    )
    ScaleA=true
    ScaleRGB=true
   )
   &GravityForce(Gravity=&RandomRangeVector(Minimum=<0 0 -980>, Maximum=<0 0 -700>))
   &SolveForcesAndVelocity(Clamp Velocity=false, Limit Acceleration=false)
   &ScaleColor(
    Scale RGB=&MakeVector(X=1, Y=1, Z=&Lerp_Float(A=1, B=0, Alpha=$System.TOD))
    ScaleA=true
    ScaleRGB=true
   ) {module=ScaleColor001}
  ]
  rendererClasses [&NiagaraSpriteRendererProperties]
  renderers [{rendererClass=&NiagaraSpriteRendererProperties}]
 }
}
summary {
 systemName=NS_Emote_SaltySumo_SaltThrow
 emitters [
  {emitterName=SimpleSpriteBurst, simTarget=GPUComputeSim, rendererClasses [&NiagaraSpriteRendererProperties]}
  {emitterName=SimpleSpriteBurst001, rendererClasses [&NiagaraSpriteRendererProperties]}
  {emitterName=SimpleSpriteBurst002, rendererClasses [&NiagaraSpriteRendererProperties]}
  {emitterName=SimpleSpriteBurst003, rendererClasses [&NiagaraSpriteRendererProperties]}
 ]
 userVariables [
  User.bIsFrontEnd {type=NiagaraBool}
  User.bIsFrontEndPreview {type=NiagaraBool}
  User.Linear Color {type=LinearColor, description="", defaultValue {struct=LinearColor, value=<rgba 1 1 1 1>}}
  User.Offset {type=Vector3f, description="", defaultValue {struct=Vector3f, value=<0 0 0>}}
  User.ParaGlideLeanAlpha {type=NiagaraFloat, defaultValue=0.5, description=""}
  User.Test {type=Vector3f, description="", defaultValue {struct=Vector3f, value=<6 6 6>}}
 ]
}
